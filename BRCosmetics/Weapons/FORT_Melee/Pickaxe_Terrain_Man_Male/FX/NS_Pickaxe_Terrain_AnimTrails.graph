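# Niagara system for the swing trails of the Terrain pickaxe, drawing three pink-toned ribbons on the weapon.
# Each emitter spawns by distance moved and samples its own pair of bones or sockets on the User.SkeletalMesh.
# The pair feeds a local AnimTrail_Spawn_Location module, and their difference sets ribbon facing, keeping each ribbon aligned between the two points.
# Particle lifetime shrinks with the system's normalized loop age, so trails shorten over each loop.
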NiagaraSystem /BRCosmetics/Weapons/FORT_Melee/Pickaxe_Terrain_Man_Male/FX/NS_Pickaxe_Terrain_AnimTrails
dependencies {
 usedDataInterfaces []
 usedDynamicInputs []
 usedModules [
  &self:SystemSpawnScript.SystemScriptSource.SystemScriptGraph.NiagaraNodeAssignment_0.SetVariables_#1
  &SystemState
 ]
}
emitters {
 AnimTrail_Pink_Bottom {
  emitterUpdateScript [
   &EmitterState()
   &SpawnPerUnit(Module.Use Max Movement Threshold=true, Module.Use Movement Tolerance=false, Use Spawn Probability=false)
  ]
  particleSpawnScript [
   &InitializeRibbon(
    Color=<rgba 1 0.625 0.5 1>
    Lifetime=&Multiply_Float(A=0.165, B=&OneMinusFloat(Float=$System.NormalizedLoopAge))
    Module.Write Color=true
    Module.Write Lifetime=true
    Module.Write Mass=false
    Module.Write Material Random=true
    Module.Write Position=false
    Module.Write RibbonTwist=false
    Module.Write RibbonWidth=true
    Ribbon Width=30
   )
   &SkeletalMeshLocation(
    Apply To Particle Position=false
    Bone / Socket Index=2
    Inherit Velocity=false
    MaskSpawnGroup=false
    Mesh Sampling Type=::NewEnumerator5
    Skeletal Mesh=$User.SkeletalMesh
    Socket Sampling Mode=::NewEnumerator5
    Use Sampled Mesh Attribute as Ribbon ID=false
   )
   &SkeletalMeshLocation(
    Apply To Particle Position=false
    Bone / Socket Index=3
    Inherit Velocity=false
    MaskSpawnGroup=false
    Mesh Sampling Type=::NewEnumerator5
    Skeletal Mesh=$User.SkeletalMesh
    Socket Sampling Mode=::NewEnumerator5
    Use Sampled Mesh Attribute as Ribbon ID=false
   ) {module=SkeletalMeshLocation001}
   &self:AnimTrail_Spawn_Location(Location A:Vector3f=$Output.SkeletalMeshLocation.SampledPosition, Location B:Vector3f=$Output.SkeletalMeshLocation001.SampledPosition) {module=AnimTrail_Spawn_Location}
   /SetVariables(Particles.RibbonFacing=&Subtract_Vector(A=$Output.SkeletalMeshLocation.SampledPosition, B=$Output.SkeletalMeshLocation001.SampledPosition)) {isSetParameters=true, module=SetVariables_#2}
  ]
  particleUpdateScript [
   &ParticleState()
   &ScaleRibbonWidth()
  ]
  rendererClasses [&NiagaraRibbonRendererProperties]
  renderers [{rendererClass=&NiagaraRibbonRendererProperties}]
 }
 AnimTrail_Pink_Bottom001 {
  emitterUpdateScript [
   &EmitterState()
   &SpawnPerUnit(Module.Use Max Movement Threshold=true, Module.Use Movement Tolerance=false, Use Spawn Probability=false)
  ]
  particleSpawnScript [
   &InitializeRibbon(
    Color=<rgba 1 0 0.6864 1>
    Lifetime=&Multiply_Float(A=0.165, B=&OneMinusFloat(Float=$System.NormalizedLoopAge))
    Module.Write Color=true
    Module.Write Lifetime=true
    Module.Write Mass=false
    Module.Write Material Random=true
    Module.Write Position=false
    Module.Write RibbonTwist=false
    Module.Write RibbonWidth=true
    Ribbon Width=20
   )
   &SkeletalMeshLocation(
    Apply To Particle Position=false
    Bone / Socket Index=4
    Inherit Velocity=false
    MaskSpawnGroup=false
    Mesh Sampling Type=::NewEnumerator5
    Skeletal Mesh=$User.SkeletalMesh
    Socket Sampling Mode=::NewEnumerator5
    Use Sampled Mesh Attribute as Ribbon ID=false
   )
   &SkeletalMeshLocation(
    Apply To Particle Position=false
    Bone / Socket Index=5
    Inherit Velocity=false
    MaskSpawnGroup=false
    Mesh Sampling Type=::NewEnumerator5
    Skeletal Mesh=$User.SkeletalMesh
    Socket Sampling Mode=::NewEnumerator5
    Use Sampled Mesh Attribute as Ribbon ID=false
   ) {module=SkeletalMeshLocation001}
   &self:AnimTrail_Spawn_Location(Location A:Vector3f=$Output.SkeletalMeshLocation.SampledPosition, Location B:Vector3f=$Output.SkeletalMeshLocation001.SampledPosition) {module=AnimTrail_Spawn_Location}
   /SetVariables(Particles.RibbonFacing=&Subtract_Vector(A=$Output.SkeletalMeshLocation.SampledPosition, B=$Output.SkeletalMeshLocation001.SampledPosition)) {isSetParameters=true, module=SetVariables_#2}
  ]
  particleUpdateScript [
   &ParticleState()
   &ScaleRibbonWidth()
  ]
  rendererClasses [&NiagaraRibbonRendererProperties]
  renderers [{rendererClass=&NiagaraRibbonRendererProperties}]
 }
 AnimTrail_Pink_Top {
  emitterUpdateScript [
   &EmitterState()
   &SpawnPerUnit(Module.Use Max Movement Threshold=true, Module.Use Movement Tolerance=false, Use Spawn Probability=false)
  ]
  particleSpawnScript [
   &InitializeRibbon(
    Color=<rgba 1 0 0.6864 1>
    Lifetime=&Multiply_Float(A=0.165, B=&OneMinusFloat(Float=$System.NormalizedLoopAge))
    Module.Write Color=true
    Module.Write Lifetime=true
    Module.Write Mass=false
    Module.Write Material Random=true
    Module.Write Position=false
    Module.Write RibbonTwist=false
    Module.Write RibbonWidth=true
    Ribbon Width=20
   )
   &SkeletalMeshLocation(
    Apply To Particle Position=false
    Inherit Velocity=false
    MaskSpawnGroup=false
    Mesh Sampling Type=::NewEnumerator5
    Skeletal Mesh=$User.SkeletalMesh
    Socket Sampling Mode=::NewEnumerator5
    Use Sampled Mesh Attribute as Ribbon ID=false
   )
   &SkeletalMeshLocation(
    Apply To Particle Position=false
    Bone / Socket Index=1
    Inherit Velocity=false
    MaskSpawnGroup=false
    Mesh Sampling Type=::NewEnumerator5
    Skeletal Mesh=$User.SkeletalMesh
    Socket Sampling Mode=::NewEnumerator5
    Use Sampled Mesh Attribute as Ribbon ID=false
   ) {module=SkeletalMeshLocation001}
   &self:AnimTrail_Spawn_Location(
    Location A:Vector3f=&Add_Vector(A=$Output.SkeletalMeshLocation.SampledPosition, B=<0 0 0>)
    Location B:Vector3f=&Add_Vector(A=$Output.SkeletalMeshLocation001.SampledPosition, B=<0 0 0>)
   ) {module=AnimTrail_Spawn_Location}
   /SetVariables(Particles.RibbonFacing=&Subtract_Vector(A=$Output.SkeletalMeshLocation.SampledPosition, B=$Output.SkeletalMeshLocation001.SampledPosition)) {isSetParameters=true, module=SetVariables_#2}
  ]
  particleUpdateScript [
   &ParticleState()
   &ScaleRibbonWidth(Ribbon Width Scale=2)
  ]
  rendererClasses [&NiagaraRibbonRendererProperties]
  renderers [{rendererClass=&NiagaraRibbonRendererProperties}]
 }
}
summary {
 systemName=NS_Pickaxe_Terrain_AnimTrails
 emitters [
  {emitterName=AnimTrail_Pink_Top, rendererClasses [&NiagaraRibbonRendererProperties]}
  {emitterName=AnimTrail_Pink_Bottom, rendererClasses [&NiagaraRibbonRendererProperties]}
  {emitterName=AnimTrail_Pink_Bottom001, rendererClasses [&NiagaraRibbonRendererProperties]}
 ]
 userVariables [User.SkeletalMesh {
   type=NiagaraDataInterfaceSkeletalMesh
   defaultValue {
    struct=NiagaraExt_VariableValue_DataInterface
    value {dataInterface=&self:NiagaraDataInterfaceSkeletalMesh_0}
   }
  }]
}
